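# Physics asset for the Zaku T1 vehicle body cosmetic, giving the body mesh a simplified collision shape.
# It holds a single body on the root bone, built from one box offset forward and upward from the bone origin.
# There are no constraints, so the body has no articulated or separately simulated parts.
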PhysicsAsset /VehicleCosmetics/Bodies/Zaku/Zaku_T1/Meshes/Body_ZakuT1_PhysicsDM
bodies {root [{shapeType=Box, extentX=444, extentY=216, extentZ=116, shapeName=root_box, center=<24.5 0 84>}]}
constraints []
